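# Shared NPC conversation for the Show Item service, letting Bushranger and Polished Jade react to a player's equipped Sprite.
# The choice requires the service tag and no completed tag on the NPC, then uses per-NPC contextual lines keyed on that Sprite.
# NPC-and-Sprite-gated Continue branches burst a cue, swap outfit, apply a GameplayEffect and grant a reward.
# Bushranger gives a vault key and changes time of day; Polished Jade gives a weapon.
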ConversationDatabase /NPCLibraryCh6/NPCs/_Conversations/Shared_Conversation_ShowItem_Service
entry_points [{tag=Conversation.Entry.BasicServices, destinations [#1]}]
nodes [
 ConversationEntryPointNode_0 {
  id=#1
  class=ConversationEntryPointNode
  outputs [#2]
  properties {entryTag {tagName=Conversation.Entry.BasicServices}}
 }
 Speech_C_3 {
  id=#2
  class=Speech_C
  outputs [#3, #4]
  properties {
   bUsePerNPC_ContextualMessageLookup=true
   softTaskIcon=&T_UI_NPC_Conversation_Give
   generalConfig {defaultMessage=You want to show me something?}
   speakerEntryTagToConfig {
    "(TagName=\"Conversation.Entry.Bushranger\")" {defaultMessage=You have secret-showings for Bushy?}
    "(TagName=\"Conversation.Entry.PolishedJade\")" {defaultMessage=You have something to show me?}
   }
  }
  sub_nodes [
   FortConversationRequirement_HasService_0 {class=FortConversationRequirement_HasService, properties {serviceTag {tagName=Tandem.Service.ShowItem}}}
   FortConversationChoiceWithPreviewNode_0 {
    class=FortConversationChoiceWithPreviewNode
    properties {
     bShowPreviewOnlyIfChildTaskNodeValid=false
     defaultChoiceDisplayText=Show Item
     choiceTags {gameplayTags [{tagName=Conversation.Choice.CustomSortPriority.High}]}
    }
   }
   FortConversationRequirement_ControllerRequirement_0 {
    class=FortConversationRequirement_ControllerRequirement
    properties {
     nodeName=Require: Has GameplayTag
     participantId {tagName=Conversation.Participant.NPC}
     requirement {class=FortControllerRequirement_HasOwnedTag, properties {bInverseCondition=true, ownedTag {tagName=Tandem.Service.ShowItem.Completed}}}
    }
   }
  ]
 }
 Speech_C_0 {
  id=#4
  class=Speech_C
  outputs [#5, #6, #7]
  properties {
   bUsePerNPC_ContextualMessageLookup=true
   softTaskIcon=&T_UI_NPC_Conversation_Give
   generalConfig {defaultMessage="It is one of those creatures, right?"}
   speakerEntryTagToConfig {
    "(TagName=\"Conversation.Entry.Bushranger\")" {
     defaultMessage="Aw, not a spritefriend... But, nicething usefulgood! "
     contextualMessages [
      {
       message="Bushy and spritefriend make Nightday into Daynight! YAY!"
       requirementMatchPolicy=RequireAll
       contextRequirements [{
         participantId {tagName=Conversation.Participant.Player}
         requirement {class=FortControllerRequirement_HasOwnedTag, properties {ownedTag {tagName=Athena.Player.Sprite.Equipped}}}
        }]
       metadataParameters []
       metadataTags {gameplayTags []}
       priority {registryType=None, value=0, curve {curveTable=null, rowName=None}}
       weight {registryType=None, value=1, curve {curveTable=null, rowName=None}}
      }
      {
       message="Aw, not a spritefriend... But, nicething usefulgood! "
       requirementMatchPolicy=RequireAll
       contextRequirements [{
         participantId {tagName=Conversation.Participant.Player}
         requirement {class=FortControllerRequirement_HasOwnedTag, properties {bInverseCondition=true, ownedTag {tagName=Athena.Player.Sprite.Equipped}}}
        }]
       metadataParameters []
       metadataTags {gameplayTags []}
       priority {registryType=None, value=0, curve {curveTable=null, rowName=None}}
       weight {registryType=None, value=1, curve {curveTable=null, rowName=None}}
      }
     ]
    }
    "(TagName=\"Conversation.Entry.PolishedJade\")" {
     defaultMessage=I can hear them... The sprites...
     contextualMessages [
      {
       message="Oh! I suddenly feel... wonderful!"
       requirementMatchPolicy=RequireAll
       contextRequirements [{
         participantId {tagName=Conversation.Participant.Player}
         requirement {class=FortControllerRequirement_HasOwnedTag, properties {ownedTag {tagName=Athena.Player.Sprite.Equipped}}}
        }]
       metadataParameters []
       metadataTags {gameplayTags []}
       priority {registryType=None, value=0, curve {curveTable=null, rowName=None}}
       weight {registryType=None, value=1, curve {curveTable=null, rowName=None}}
      }
      {
       message="Nice loadout. You look ready for adventure!"
       requirementMatchPolicy=RequireAll
       contextRequirements [{
         participantId {tagName=Conversation.Participant.Player}
         requirement {class=FortControllerRequirement_HasOwnedTag, properties {bInverseCondition=true, ownedTag {tagName=Athena.Player.Sprite.Equipped}}}
        }]
       metadataParameters []
       metadataTags {gameplayTags []}
       priority {registryType=None, value=0, curve {curveTable=null, rowName=None}}
       weight {registryType=None, value=1, curve {curveTable=null, rowName=None}}
      }
     ]
    }
   }
  }
  sub_nodes [FortConversationChoiceWithPreviewNode_2 {
    class=FortConversationChoiceWithPreviewNode
    properties {
     bShowPreviewOnlyIfChildTaskNodeValid=false
     defaultChoiceDisplayText=Show Item
     choiceTags {gameplayTags [{tagName=Conversation.Choice.CustomSortPriority.Highest}]}
    }
   }]
 }
 Speech_C_2 {
  id=#7
  class=Speech_C
  outputs [#8]
  properties {nodeName="", generalConfig {defaultMessage="That sprite... made me feel better! I have to know how..."}}
  sub_nodes [
   Requirement_Conversation_NPC_C_0 {class=Requirement_Conversation_NPC_C, properties {requireExactMatch=true, conversationEntryTag {tagName=Conversation.Entry.PolishedJade}}}
   ConversationSideEffect_Client_BurstGameplayCue_C_2 {
    class=ConversationSideEffect_Client_BurstGameplayCue_C
    properties {
     bShowPropertyDetails=false
     conversationParticipant=NPC
     nodeName=Burst GCN - Local
     gameplayCueTag {tagName=GameplayCue.FirePetal.DayNight.TransitionLocal}
    }
   }
   ConversationSideEffect_Server_ChangeNPCOutfit_C_0 {
    class=ConversationSideEffect_Server_ChangeNPCOutfit_C
    properties {
     bShowPropertyDetails=false
     newOutfit=/BRCosmetics/Athena/Items/Cosmetics/Characters/Character_PolishedJade_Mind
     nodeName=NPC Outfit Change
     requireExactMatch=true
     conversationEntryTag {tagName=Conversation.Entry.PolishedJade}
     outfitVariantData [{
       cachedItemType=Max_None
       item=&Character_PolishedJade_Mind
       channelInfoList [{
         customData=""
         itemVariantIsUsedFor=&Character_PolishedJade_Mind
         activeVariantTag {tagName=Cosmetics.Variant.Property.Stage3}
         ownedVariantTags {gameplayTags []}
         variantChannelTag {tagName=Cosmetics.Variant.Channel.Progressive}
        }]
      }]
    }
   }
   Requirement_Conversation_Equipped_Item_C_0 {class=Requirement_Conversation_Equipped_Item_C, properties {requiredItemTag {tagName=Athena.Player.Sprite.Equipped}}}
   ConversationChoiceNode_0 {class=ConversationChoiceNode, properties {defaultChoiceDisplayText=Continue}}
   ConversationSideEffect_Server_GiveItemFromWID_C_2 {
    class=ConversationSideEffect_Server_GiveItemFromWID_C
    properties {
     bShowPropertyDetails=false
     itemDefinition=/FirePetalWeaponGameplay/Gameplay/FirePetal_AR_Fast/WID_Assault_FirePetal_Fast_Athena_VR
     nodeName=Give Item - Epic Weapon
    }
   }
   ConversationSideEffect_Client_ApplyGameplayEffect_C_1 {
    class=ConversationSideEffect_Client_ApplyGameplayEffect_C
    properties {
     bShowPropertyDetails=false
     conversationParticipant=NPC
     gameplayEffect=/NPCLibrary/NPCs/_GEs/GE_CheckCosmeticSwap.GE_CheckCosmeticSwap_C
     nodeName=Apply Gameplay Effect - Service Tag
    }
   }
  ]
 }
 Continue_C_2 {id=#8, class=Continue_C, properties {softTaskIcon=/Game/Athena/HUD/Interaction/T_UI_NPC_statement}}
 Speech_C_4 {
  id=#6
  class=Speech_C
  outputs [#8]
  properties {nodeName="", generalConfig {defaultMessage="Nightday begone! Daynight shines on Bushranger friendfriend!"}}
  sub_nodes [
   ConversationSideEffect_Client_BurstGameplayCue_C_0 {
    class=ConversationSideEffect_Client_BurstGameplayCue_C
    properties {
     bShowPropertyDetails=false
     conversationParticipant=NPC
     nodeName=Burst GCN - Local
     gameplayCueTag {tagName=GameplayCue.FirePetal.DayNight.TransitionLocal}
    }
   }
   ConversationSideEffect_Server_ChangeNPCOutfit_C_1 {
    class=ConversationSideEffect_Server_ChangeNPCOutfit_C
    properties {
     bShowPropertyDetails=false
     newOutfit=/BRCosmetics/Athena/Items/Cosmetics/Characters/CID_611_Athena_Commando_M_WeepingWoods
     nodeName=NPC Outfit Change
     requireExactMatch=true
     conversationEntryTag {tagName=Conversation.Entry.Bushranger}
     outfitVariantData [{
       cachedItemType=Max_None
       item=&CID_611_Athena_Commando_M_WeepingWoods
       channelInfoList [{
         customData=""
         itemVariantIsUsedFor=&CID_611_Athena_Commando_M_WeepingWoods
         activeVariantTag {tagName=Cosmetics.Variant.Property.Stage2}
         ownedVariantTags {gameplayTags []}
         variantChannelTag {tagName=Cosmetics.Variant.Channel.Parts}
        }]
      }]
    }
   }
   Requirement_Conversation_NPC_C_2 {class=Requirement_Conversation_NPC_C, properties {requireExactMatch=true, conversationEntryTag {tagName=Conversation.Entry.Bushranger}}}
   Requirement_Conversation_Equipped_Item_C_2 {class=Requirement_Conversation_Equipped_Item_C, properties {requiredItemTag {tagName=Athena.Player.Sprite.Equipped}}}
   ConversationChoiceNode_11 {class=ConversationChoiceNode, properties {defaultChoiceDisplayText=Continue}}
   ConversationSideEffect_Server_GiveItemFromWID_C_0 {
    class=ConversationSideEffect_Server_GiveItemFromWID_C
    properties {
     bShowPropertyDetails=false
     itemDefinition=/FirePetalDayNightForest/Gameplay/KeyItem/WID_FirePetal_DayNight_VaultKey
     nodeName=Give Item - Vault Key
    }
   }
   ConversationSideEffect_ChangeTimeOfDay_C_0 {class=ConversationSideEffect_ChangeTimeOfDay_C, properties {bShowPropertyDetails=false, nodeName=Change Time of Day Event}}
   ConversationSideEffect_Client_ApplyGameplayEffect_C_2 {
    class=ConversationSideEffect_Client_ApplyGameplayEffect_C
    properties {
     bShowPropertyDetails=false
     conversationParticipant=NPC
     gameplayEffect=&GE_CheckCosmeticSwap_C
     nodeName=Apply Gameplay Effect - Service Tag
    }
   }
  ]
 }
 ReturnToStart_C_1 {
  id=#5
  class=ReturnToStart_C
  properties {softTaskIcon=&T_UI_NPC_Back}
  sub_nodes [ConversationChoiceNode_6 {class=ConversationChoiceNode, properties {defaultChoiceDisplayText=Back}}]
 }
 Back_C_0 {
  id=#3
  class=Back_C
  sub_nodes [ConversationChoiceNode_20 {
    class=ConversationChoiceNode
    properties {defaultChoiceDisplayText=Not really..., choiceTags {gameplayTags [{tagName=Conversation.Choice.CustomSortPriority.Highest}]}}
   }]
 }
]
